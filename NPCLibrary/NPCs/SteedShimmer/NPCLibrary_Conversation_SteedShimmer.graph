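# Conversation database for the SteedShimmer NPC, providing the greeting and top-level branches when a player talks to them.
# The entry point leads to one Speech node, whose default line can give way to contextual lines gated on first-ever meeting, per-match conversation count or worn outfit.
# First-meeting and outfit reactions take their priority from named curve rows.
# Speech then branches through link nodes to shared quest, basic service and random service conversations, plus SteedShimmer's own Dialogue entry.
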ConversationDatabase /NPCLibrary/NPCs/SteedShimmer/NPCLibrary_Conversation_SteedShimmer
entry_points [{tag=Conversation.Entry.SteedShimmer, destinations [#1]}]
nodes [
 ConversationEntryPointNode_7 {
  id=#1
  class=ConversationEntryPointNode
  outputs [#2]
  properties {entryTag {tagName=Conversation.Entry.SteedShimmer}}
 }
 Speech_C_4 {
  id=#2
  class=Speech_C
  outputs [#3, #4, #5, #6]
  properties {
   nodeName=""
   generalConfig {
    defaultMessage="Unlike these other NPCs, *I* have free will."
    contextualMessages [
     {
      message=It's D-A-S-H-A. I want the news to spell it right this time.
      contextRequirements [{requirement {class=FortControllerRequirement_IsFirstEverConversationWithNPC}}]
      priority {curve {rowName=Tandem.Speech.Priority.FirstEverInteraction}}
     }
     {
      message=You sure you wanna mess with this?
      contextRequirements [{requirement {class=FortControllerRequirement_CheckNPCMatchConversationCount, properties {nPCUniqueIdTag {tagName=AISpawnerData.Type.Tandem.SteedShimmer}}}}]
      priority {curve {rowName=None}}
     }
     {
      message=I'd abide by the laws if the laws were any fun.
      contextRequirements [{requirement {class=FortControllerRequirement_CheckNPCMatchConversationCount, properties {nPCUniqueIdTag {tagName=AISpawnerData.Type.Tandem.SteedShimmer}}}}]
      priority {curve {rowName=None}}
     }
     {
      message="Unlike these other NPCs, *I* have free will."
      contextRequirements [{requirement {class=FortControllerRequirement_CheckNPCMatchConversationCount, properties {nPCUniqueIdTag {tagName=AISpawnerData.Type.Tandem.SteedShimmer}}}}]
      priority {curve {rowName=None}}
     }
     {
      message=So that's what freezer burn looks like. Gross.
      contextRequirements [{requirement {
         class=FortControllerRequirement_HasCID
         properties {allowedCIds [
           /BRCosmetics/Athena/Items/Cosmetics/Characters/CID_288_Athena_Commando_M_IceKing
           /BRCosmetics/Athena/Items/Cosmetics/Characters/Character_BriskReign
          ]}
        }}]
      priority {curve {rowName=Tandem.Speech.Priority.OutfitReaction}}
     }
     {
      message="The Ice King sends his regards!"
      contextRequirements [{requirement {
         class=FortControllerRequirement_HasCID
         properties {allowedCIds [
           /BRCosmetics/Athena/Items/Cosmetics/Characters/Character_FoundKelp_Bolt
           /BRCosmetics/Athena/Items/Cosmetics/Characters/CID_A_286_Athena_Commando_M_Turtleneck
          ]}
        }}]
      priority {curve {rowName=Tandem.Speech.Priority.OutfitReaction}}
     }
     {
      message=Admit it. You miss me.
      contextRequirements [{requirement {
         class=FortControllerRequirement_HasCID
         properties {allowedCIds [
           /BRCosmetics/Athena/Items/Cosmetics/Characters/Character_CozyMechanic_Hen
           /BRCosmetics/Athena/Items/Cosmetics/Characters/Character_MechanicalEngineerRev
           /BRCosmetics/Athena/Items/Cosmetics/Characters/CID_770_Athena_Commando_F_MechanicalEngineer
           /BRCosmetics/Athena/Items/Cosmetics/Characters/CID_A_127_Athena_Commando_F_MechanicalEngineerSummer
           /BRCosmetics/Athena/Items/Cosmetics/Characters/CID_A_077_Athena_Commando_F_ArmoredEngineer
          ]}
        }}]
      priority {curve {rowName=Tandem.Speech.Priority.OutfitReaction}}
     }
    ]
   }
  }
 }
 ConversationLinkNode_3 {
  id=#6
  class=ConversationLinkNode
  properties {remoteEntryTag {tagName=Conversation.Entry.SteedShimmer.Dialogue}}
 }
 ConversationLinkNode_2 {id=#5, class=ConversationLinkNode, properties {remoteEntryTag {tagName=Conversation.Entry.RandomServices}}}
 ConversationLinkNode_0 {id=#4, class=ConversationLinkNode, properties {remoteEntryTag {tagName=Conversation.Entry.BasicServices}}}
 ConversationLinkNode_1 {id=#3, class=ConversationLinkNode, properties {remoteEntryTag {tagName=Conversation.Entry.NPCLibrary.Quests}}}
]
