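# Conversation database for the SteedShimmer NPC's "Ask about..." dialogue, a lore chat in the character's voice.
# A single tagged entry point leads to an opening speech that offers three topics (The Imagined Order, Jules, The Ice King) plus a Back node.
# Each topic plays a linear chain of four Speech nodes, and all three converge on one shared ReturnToStart node labeled BACK.
# The graph holds only speech and navigation, with no conditions or rewards.
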ConversationDatabase /NPCLibrary/NPCs/SteedShimmer/NPCLibrary_Conversation_SteedShimmer_Dialogue
entry_points [{tag=Conversation.Entry.SteedShimmer.Dialogue, destinations [#1]}]
nodes [
 ConversationEntryPointNode_0 {
  id=#1
  class=ConversationEntryPointNode
  outputs [#2]
  properties {entryTag {tagName=Conversation.Entry.SteedShimmer.Dialogue}}
 }
 Speech_C_0 {
  id=#2
  class=Speech_C
  outputs [#3, #4, #5, #6]
  properties {nodeName="", softTaskIcon=&T_UI_NPC_question, generalConfig {defaultMessage="Nosy! I kinda like it."}}
  sub_nodes [ConversationChoiceNode_5 {class=ConversationChoiceNode, properties {defaultChoiceDisplayText=Ask about...}}]
 }
 Speech_C_6 {
  id=#6
  class=Speech_C
  outputs [#7]
  properties {
   nodeName=""
   softTaskIcon=&T_UI_NPC_question
   generalConfig {defaultMessage="This Island's become beef city. Everybody's beefin' with everyone. "}
  }
  sub_nodes [ConversationChoiceNode_1 {class=ConversationChoiceNode, properties {defaultChoiceDisplayText=The Ice King}}]
 }
 Speech_C_8 {
  id=#7
  class=Speech_C
  outputs [#8]
  properties {
   nodeName=""
   softTaskIcon=&T_UI_NPC_question
   generalConfig {defaultMessage="I can't let these people have all the fun without me, can I?!"}
  }
 }
 Speech_C_9 {
  id=#8
  class=Speech_C
  outputs [#9]
  properties {
   nodeName=""
   softTaskIcon=&T_UI_NPC_question
   generalConfig {defaultMessage=And if it comes down to the Ice King or the Foundation...}
  }
 }
 Speech_C_10 {
  id=#9
  class=Speech_C
  outputs [#10]
  properties {
   nodeName=""
   softTaskIcon=&T_UI_NPC_question
   generalConfig {defaultMessage="One of those guys is slightly less annoying than the other. "}
  }
 }
 ReturnToStart_C_23 {
  id=#10
  class=ReturnToStart_C
  properties {softTaskIcon=&T_UI_NPC_Back}
  sub_nodes [ConversationChoiceNode_32 {class=ConversationChoiceNode, properties {defaultChoiceDisplayText=BACK}}]
 }
 Speech_C_3 {
  id=#5
  class=Speech_C
  outputs [#11]
  properties {
   nodeName=""
   softTaskIcon=&T_UI_NPC_question
   generalConfig {defaultMessage="She is SUCH a goody-two-shoes. She loves \"following rules\" and \"serving the greater good...\""}
  }
  sub_nodes [ConversationChoiceNode_2 {class=ConversationChoiceNode, properties {defaultChoiceDisplayText=Jules}}]
 }
 Speech_C_4 {
  id=#11
  class=Speech_C
  outputs [#12]
  properties {
   nodeName=""
   softTaskIcon=&T_UI_NPC_question
   generalConfig {defaultMessage="...and \"maintaining a standard of ethics for her work.\" SNOOZE!"}
  }
 }
 Speech_C_5 {
  id=#12
  class=Speech_C
  outputs [#13]
  properties {
   nodeName=""
   softTaskIcon=&T_UI_NPC_question
   generalConfig {defaultMessage=She was way more adventurous when we were younger.}
  }
 }
 Speech_C_2 {
  id=#13
  class=Speech_C
  outputs [#10]
  properties {
   nodeName=""
   softTaskIcon=&T_UI_NPC_question
   generalConfig {defaultMessage="It's my duty as her former friend to shake it up a little bit, dontcha think?"}
  }
 }
 Back_C_0 {
  id=#4
  class=Back_C
  sub_nodes [ConversationChoiceNode_7 {class=ConversationChoiceNode, properties {defaultChoiceDisplayText=BACK}}]
 }
 Speech_C_7 {
  id=#3
  class=Speech_C
  outputs [#14]
  properties {
   nodeName=""
   softTaskIcon=&T_UI_NPC_question
   generalConfig {defaultMessage="Yeah, I worked there. Good cookies in the cafeteria."}
  }
  sub_nodes [ConversationChoiceNode_0 {class=ConversationChoiceNode, properties {defaultChoiceDisplayText=The Imagined Order}}]
 }
 Speech_C_15 {
  id=#14
  class=Speech_C
  outputs [#15]
  properties {
   nodeName=""
   softTaskIcon=&T_UI_NPC_question
   generalConfig {defaultMessage="I know what brought 'em down, and it wasn't you Loopers."}
  }
 }
 Speech_C_14 {
  id=#15
  class=Speech_C
  outputs [#16]
  properties {
   nodeName=""
   softTaskIcon=&T_UI_NPC_question
   generalConfig {defaultMessage="It was bureaucracy. Paperwork, rules, red tape... And the Seven. "}
  }
 }
 Speech_C_1 {
  id=#16
  class=Speech_C
  outputs [#10]
  properties {
   nodeName=""
   softTaskIcon=&T_UI_NPC_question
   generalConfig {defaultMessage="Give me a good loadout and 20 minutes, I'll happily settle that score on IO's behalf."}
  }
 }
]
